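# Master material for the Alfredo armor player body, a skeletal mesh using material attributes with Default Lit, Clear Coat and Subsurface Profile shading.
# The shared character PreFX function supplies the base, then the SkinFX_Mask red channel, shaped by power and multiplier, adds tinted emissive.
# Static switches layer mask-driven clear coat, thin film and optional metals LUT.
# An emissive fresnel function follows, then the shared character PostFX function finishes the chain.
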Material /BRCosmetics/Characters/Player/Male/Medium/Bodies/M_MED_Alfredo_Armor/Materials/M_MED_Alfredo_Armor_MASTER
graph {
 properties {
  bUseMaterialAttributes=true
  bUsedWithMeshDeformer=true
  bUsedWithMorphTargets=true
  bUsedWithSkeletalMesh=true
  pixelDepthOffsetMode=PDOM_Legacy
  shadingModel=MSM_FromMaterialExpression
  subsurfaceProfile=&SS_HeroSkin_02
  usedShadingModels=Default Lit | Clear Coat | Subsurface Profile
 }
 nodes [
  n1 = MaterialFunctionCall() {materialFunction=&BaseFN_CharacterFunction_PreFX}
  n2 = TextureSample(Tex=TextureObjectParameter() {group=Skin FX, parameterName=SkinFX_Mask})
  n3 = GetMaterialAttributes(None=@n1.Result) {attributeGetTypes [#1]}
  n4 = VectorParameter() {parameterName=Emissive Color}
  n5 = Multiply(
   A=Power(Base=@n2.R, Exp=ScalarParameter() {defaultValue=1, group=Alfredo, parameterName=Alfredo_emissivePow, sortPriority=0}) {constExponent=5}
   B=ScalarParameter() {defaultValue=0.5, group=Alfredo, parameterName=Alfredo_emissiveMult}
  )
  n6 = SetMaterialAttributes(MaterialAttributes=@n3.MaterialAttributes, Emissive Color=Add(A=@n3.Emissive Color, B=Multiply(A=@n4.RGB, B=@n5))) {attributeSetTypes [#1]}
  n7 = StaticSwitchParameter(
   True=MaterialFunctionCall!(Alpha=StaticComponentMaskParameter(None=@n2.RGB) {defaultR=true, group=Skin FX, parameterName=Clear Coat Channel}, In=@n6) {materialFunction=/Game/Weapons/WeaponSkins/MaterialFunction/Clearcoat}.Result
   False=@n6
  ) {defaultValue=true, group=Skin FX, parameterName=Use Clear Coat}
  n8 = StaticSwitchParameter(
   True=MaterialFunctionCall!(Alpha=StaticComponentMaskParameter(None=@n2.RGB) {defaultR=true, group=Skin FX, parameterName=ThinFilm_Channel}, Material Attributes=@n7) {materialFunction=/Game/Weapons/WeaponSkins/MaterialFunction/ThinFilmFunction}.Result
   False=@n7
  ) {defaultValue=true, group=Skin FX, parameterName=Use Thin Film}
  n9 = MaterialFunctionCall(Mask=@n2.RGBA, Materials=@n8) {materialFunction=&MF_MetalsLUT}
  MaterialFunctionCall(
   AO=@n1.AO
   Distance Blend=@n1.AO
   Fresnel=@n1.AO
   Gmap=@n1.AO
   Pre Skinned Local Position=@n1.AO
   Material=MaterialFunctionCall!(In=StaticSwitchParameter(True=@n9.Result, False=@n8) {group=Skin FX, parameterName=UseMetalLUT}) {materialFunction=/Game/Weapons/WeaponSkins/MaterialFunction/MF_EmissiveFresnel}.Result
  ) {materialFunction=&BaseFN_CharacterFunction_PostFX}
 ]
}
